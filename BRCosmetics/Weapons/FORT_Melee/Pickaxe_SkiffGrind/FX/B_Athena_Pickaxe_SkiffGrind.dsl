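; Blueprint for the SkiffGrind pickaxe cosmetic in Battle Royale, controlling its idle Niagara effect.
; When weapon visibility changes, it activates the idle effect component if visible, without resetting it, and deactivates it otherwise.
; The local-controller-only flag is not used, and the construction script only calls the parent.
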(event EventOnWeaponVisibilityChanged
  (params (bVisible :type Boolean) (bSetForLocalControllerOnly :type Boolean))
  (bind _ns_pickaxe_skiffgrind_idle (Variables|Default|GetNS_Pickaxe_SkiffGrind_Idle))
  (if bVisible
    (Components|Activation|Activate _ns_pickaxe_skiffgrind_idle false)
    (else
      (Components|Activation|Deactivate _ns_pickaxe_skiffgrind_idle))))

(fn ConstructionScript
  (|CallParentFunction))
